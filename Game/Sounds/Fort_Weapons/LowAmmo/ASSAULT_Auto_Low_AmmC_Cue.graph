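# Sound cue that plays the looping low-ammo layer for automatic assault rifles.
# A looping wave player of the auto rifle low-ammo loop feeds a Modulator root with pitch randomization flattened to unity.
# It routes to the degraded low-ammo ranged weapon sound class.
# A concurrency override allows one instance per owner and prevents new ones, so the loop does not stack.
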SoundCue /Game/Sounds/Fort_Weapons/LowAmmo/ASSAULT_Auto_Low_AmmC_Cue.ASSAULT_Auto_Low_Ammc_Cue
root=Modulator_0
nodes [
 {id=Modulator_0, children [WavePlayer_0], properties {pitchMax=1, pitchMin=1}}
 {id=WavePlayer_0, properties {bLooping=true, soundWaveAssetPtr=/Game/Sounds/Fort_Weapons/AutoRifle/fort_auto_LowAmmoLoop}}
]
properties {
 bOverrideConcurrency=true
 maxDistance=2097152
 soundClassObject=/Game/Sounds/FortSoundClassesAndModes/Fort_WeaponRanged_LowAmmo_Degraded_PSM
 volumeMultiplier=1
 concurrencyOverrides {bLimitToOwner=true, maxCount=1, resolutionRule=PreventNew}
}
