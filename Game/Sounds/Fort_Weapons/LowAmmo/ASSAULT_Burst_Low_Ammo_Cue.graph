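# Sound cue for the low-ammo sound of a three-round burst assault weapon.
# A single wave player feeds a Modulator that pins pitch at 1, so playback pitch does not vary.
# It routes through the degraded low-ammo ranged weapon sound class and overrides concurrency to prevent new instances once the limit is reached.
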SoundCue /Game/Sounds/Fort_Weapons/LowAmmo/ASSAULT_Burst_Low_Ammo_Cue
root=Modulator_0
nodes [
 {id=Modulator_0, children [WavePlayer_1], properties {pitchMax=1, pitchMin=1}}
 {id=WavePlayer_1, properties {soundWaveAssetPtr=/Game/Sounds/Fort_Weapons/AutoRifle/3round_burst_lowammo}}
]
properties {
 bOverrideConcurrency=true
 maxDistance=2097152
 soundClassObject=/Game/Sounds/FortSoundClassesAndModes/Fort_WeaponRanged_LowAmmo_Degraded_PSM
 volumeMultiplier=1
 concurrencyOverrides {resolutionRule=PreventNew}
}
