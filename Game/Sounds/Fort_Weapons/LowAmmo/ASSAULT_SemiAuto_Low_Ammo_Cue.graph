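# Sound cue that plays the low-ammo warning sound for semi-auto assault weapons.
# A single rifle low-ammo wave feeds a Modulator that applies a fixed downward pitch shift, with no random variation.
# It routes to the Fort_WeaponRanged_LowAmmo_Degraded_PSM sound class and overrides concurrency to prevent new instances once the limit is reached.
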SoundCue /Game/Sounds/Fort_Weapons/LowAmmo/ASSAULT_SemiAuto_Low_Ammo_Cue
root=Modulator_0
nodes [
 {id=Modulator_0, children [WavePlayer_2], properties {pitchMax=0.85, pitchMin=0.85}}
 {id=WavePlayer_2, properties {soundWaveAssetPtr=&/weapon_low_ammo_hi_rifle_05}}
]
properties {
 bOverrideConcurrency=true
 maxDistance=2097152
 soundClassObject=/Game/Sounds/FortSoundClassesAndModes/Fort_WeaponRanged_LowAmmo_Degraded_PSM
 concurrencyOverrides {resolutionRule=PreventNew}
}
