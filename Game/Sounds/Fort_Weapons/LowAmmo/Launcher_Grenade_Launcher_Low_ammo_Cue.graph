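# Sound cue for the grenade launcher's low-ammo warning in the weapon audio set.
# A Mixer layers a high rifle low-ammo wave with the quad launcher low-ammo wave, weighting the launcher layer louder.
# Each layer passes through a Modulator that fixes pitch lower without random variation.
# The cue uses the degraded ranged-weapon low-ammo sound class and overrides concurrency to prevent new instances at the limit.
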SoundCue /Game/Sounds/Fort_Weapons/LowAmmo/Launcher_Grenade_Launcher_Low_ammo_Cue
root=Mixer_0
nodes [
 {id=Mixer_0, children [Modulator_1,Modulator_0], properties {inputVolume [0.8,1.7]}}
 {id=Modulator_0, children [WavePlayer_0], properties {pitchMax=0.8, pitchMin=0.8, volumeMax=1, volumeMin=1}}
 {id=Modulator_1, children [WavePlayer_66], properties {pitchMax=0.8, pitchMin=0.8, volumeMax=1, volumeMin=1}}
 {id=WavePlayer_0, properties {soundWaveAssetPtr=/Game/Sounds/Fort_Weapons/Launchers/Quad/Weapon_LauncherQuad_LowAmmo}}
 {id=WavePlayer_66, properties {soundWaveAssetPtr=&/weapon_low_ammo_hi_rifle_02}}
]
properties {
 bOverrideConcurrency=true
 maxDistance=2097152
 soundClassObject=/Game/Sounds/FortSoundClassesAndModes/Fort_WeaponRanged_LowAmmo_Degraded_PSM
 volumeMultiplier=1.3
 concurrencyOverrides {resolutionRule=PreventNew}
}
