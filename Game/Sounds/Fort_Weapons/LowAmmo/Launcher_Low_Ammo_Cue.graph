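# Sound cue for the low-ammo warning of launcher weapons.
# A single wave player feeds the quad launcher low-ammo wave straight to the output at reduced volume, routed through a ranged-weapon low-ammo sound class.
# Its own concurrency override caps simultaneous instances per owner, stopping the oldest and scaling volume down as copies stack.
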SoundCue /Game/Sounds/Fort_Weapons/LowAmmo/Launcher_Low_Ammo_Cue
root=WavePlayer_0
nodes [{id=WavePlayer_0, properties {soundWaveAssetPtr=/Game/Sounds/Fort_Weapons/Launchers/Quad/Weapon_LauncherQuad_LowAmmo}}]
properties {
 bOverrideConcurrency=true
 maxDistance=2097152
 soundClassObject=/Game/Sounds/FortSoundClassesAndModes/Fort_WeaponRanged_LowAmmo_Degraded_PSM
 volumeMultiplier=0.4
 concurrencyOverrides {bLimitToOwner=true, maxCount=2, resolutionRule=StopOldest, volumeScale=0.6}
}
