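# Sound cue for the low-ammo warning sound of the hand cannon pistol.
# A single wave player plays a high rifle low-ammo sample directly, with no randomization or modulation.
# It routes to a degraded ranged-weapon low-ammo sound class and overrides concurrency so new instances are blocked once the limit is reached.
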SoundCue /Game/Sounds/Fort_Weapons/LowAmmo/Pistol_Handcannon_Low_Ammo_Cue
root=WavePlayer_67
nodes [{id=WavePlayer_67, properties {soundWaveAssetPtr=&/weapon_low_ammo_hi_rifle_08}}]
properties {
 bOverrideConcurrency=true
 maxDistance=2097152
 soundClassObject=/Game/Sounds/FortSoundClassesAndModes/Fort_WeaponRanged_LowAmmo_Degraded_PSM
 concurrencyOverrides {resolutionRule=PreventNew}
}
