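# Sound cue for the six-shooter pistol's low-ammo sound, routed to a degraded low-ammo ranged weapon sound class.
# A fixed short delay feeds a single wave player that reuses a high rifle low-ammo wave, with raised cue pitch and volume.
# Its own concurrency override blocks new instances at the limit rather than stopping playing ones.
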SoundCue /Game/Sounds/Fort_Weapons/LowAmmo/Pistol_SixShooter_Low_Ammo_Cue
root=Delay_0
nodes [
 {id=Delay_0, children [WavePlayer_67], properties {delayMax=0.1, delayMin=0.1}}
 {id=WavePlayer_67, properties {soundWaveAssetPtr=&/weapon_low_ammo_hi_rifle_08}}
]
properties {
 bOverrideConcurrency=true
 maxDistance=2097152
 pitchMultiplier=1.1
 soundClassObject=/Game/Sounds/FortSoundClassesAndModes/Fort_WeaponRanged_LowAmmo_Degraded_PSM
 volumeMultiplier=1.5
 concurrencyOverrides {resolutionRule=PreventNew}
}
