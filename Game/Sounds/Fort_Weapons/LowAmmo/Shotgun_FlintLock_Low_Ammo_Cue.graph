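# Sound cue for the flintlock shotgun's low-ammo sound.
# A Random node at the root picks between two low-ammo wave variations with equal weight.
# It routes to the degraded low-ammo ranged weapon sound class and overrides concurrency so new instances are blocked rather than cutting off playing ones.
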SoundCue /Game/Sounds/Fort_Weapons/LowAmmo/Shotgun_FlintLock_Low_Ammo_Cue
root=Random_0
nodes [
 {id=Random_0, children [WavePlayer_0,WavePlayer_1], properties {weights [1,1]}}
 {id=WavePlayer_0, properties {soundWaveAssetPtr=/Game/Sounds/Fort_Weapons/FlintLock/Flintlock_Shotgun_LowAmmo_01}}
 {id=WavePlayer_1, properties {soundWaveAssetPtr=/Game/Sounds/Fort_Weapons/FlintLock/Flintlock_Shotgun_LowAmmo_02}}
]
properties {
 bOverrideConcurrency=true
 maxDistance=2097152
 soundClassObject=/Game/Sounds/FortSoundClassesAndModes/Fort_WeaponRanged_LowAmmo_Degraded_PSM
 concurrencyOverrides {resolutionRule=PreventNew}
}
