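# Sound cue for the low-ammo sound of standard shotguns.
# A short fixed Delay feeds a single Wave Player, which plays a rifle-named high low-ammo wave rather than a shotgun-specific one.
# The cue boosts volume, uses a ranged-weapon low-ammo sound class, and overrides concurrency with a rule that prevents new instances once the limit is reached.
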SoundCue /Game/Sounds/Fort_Weapons/LowAmmo/Shotgun_Standard_Low_Ammo_Cue
root=Delay_0
nodes [
 {id=Delay_0, children [WavePlayer_66], properties {delayMax=0.05, delayMin=0.05}}
 {id=WavePlayer_66, properties {soundWaveAssetPtr=&/weapon_low_ammo_hi_rifle_02}}
]
properties {
 bOverrideConcurrency=true
 maxDistance=2097152
 soundClassObject=/Game/Sounds/FortSoundClassesAndModes/Fort_WeaponRanged_LowAmmo_Degraded_PSM
 volumeMultiplier=1.5
 concurrencyOverrides {resolutionRule=PreventNew}
}
